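# Translucent, unlit, two-sided Niagara mesh-particle material for lightning bolts on the TotalFlock backpack.
# SplineThicken shapes tapered width, adding panning-noise world position offset damped at both tips.
# Opacity combines a panning bolt texture, height-lerp noise dissolve, tip mask, view-angle fade and, on some shading paths, depth fade.
# Dynamic parameters drive width, dissolve and noise amplitude; particle color feeds the emissive adjustment function.
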Material /BRCosmetics/Accessories/FORT_Backpacks/Backpack_M_MED_TotalFlock/FX/M_SplineThicken_Lightning
graph {
 properties {
  bUsedWithNiagaraMeshParticles=true
  blendMode=BLEND_Translucent
  pixelDepthOffsetMode=PDOM_Legacy
  shadingModel=MSM_Unlit
  twoSided=true
  usedShadingModels=Unlit
 }
 nodes [
  n1 = DynamicParameter() {defaultValue=<rgba 2 0 1 1>, paramNames [Width Base Scalar, Dissolve, WPO Noise Amplitude, Param4]}
  n2 = ScalarParameter() {defaultValue=1, desc=Allows you to thicken the lightning for glow renderers., group=Spline Thicken, parameterName=Width Base Scalar}
  n3 = Constant()
  n4 = ComponentMask(None=TextureCoordinate()) {g=true}
  n5 = MaterialFunctionCall(Input=OneMinus(None=@n4), Input Low=Constant(), Input High=Constant() {r=0.5}) {materialFunction=&RemapValue_0-1}
  n6 = ScalarParameter() {defaultValue=1, group=Spline Thicken, parameterName=WidthTip_0_Exp}
  n7 = MaterialFunctionCall(Input=@n4, Input Low=Constant(), Input High=Constant() {r=0.5}) {materialFunction=&RemapValue_0-1}
  n8 = ScalarParameter() {defaultValue=1, group=Spline Thicken, parameterName=WidthTip_1_Exp}
  n9 = Saturate(None=Multiply(A=Power(Base=@n5.Result, Exp=@n6) {constExponent=1}, B=Power(Base=@n7.Result, Exp=@n8) {constExponent=3}))
  n10 = MaterialFunctionCall(Float2=TextureCoordinate()) {materialFunction=&BreakOutFloat2Components}
  n11 = MaterialFunctionCall(X=@n10.R, Y=Multiply(A=@n10.G, B=ScalarParameter() {defaultValue=1, group=WPO Noise, parameterName=WPO Noise Frequency})) {materialFunction=&MakeFloat2}
  n12 = ScalarParameter() {defaultValue=1, group=WPO Noise, parameterName=WPO Noise Pan Speed}
  n13 = Add(A=@n11.Result) {constB=0.5}
  n14 = AppendVector(
   A=TextureSampleParameter2D(UVs=Panner(Coordinate=@n11.Result, Speed=@n12) {speedY=1}) {group=WPO Noise, parameterName=WPO Noise, texture=&T_Inky_Smoke_Tile}.R
   B=TextureSampleParameter2D(UVs=Panner(Coordinate=@n13, Speed=@n12) {speedY=1}) {group=WPO Noise, parameterName=WPO Noise, texture=&T_Inky_Smoke_Tile}.R
  )
  n15 = Multiply(A=Subtract(A=@n14) {constB=0.5}) {constB=2}
  n16 = Multiply(
   A=Multiply(
    A=Multiply(A=@n9, B=@n15) {constA=1, desc=Dampen Tips}
    B=DynamicParameter() {defaultValue=<rgba 2 0 1 1>, paramNames [Width Base Scalar, Dissolve, WPO Noise Amplitude, Param4]}.WPO Noise Amplitude
   )
   B=ScalarParameter() {defaultValue=1, group=WPO Noise, parameterName=WPO Noise Amplitude}
  )
  n17 = Constant()
  Add(
   A=MaterialFunctionCall(
    WidthBase=Multiply(A=@n1.Width Base Scalar, B=@n2)
    WidthTip=ScalarParameter() {defaultValue=1, group=Spline Thicken, parameterName=Width Tip Scalar}
    UVs for Thickness=AppendVector(A=@n3, B=@n9)
   ) {materialFunction=&SplineThicken}.WorldPositionOffset
   B=Transform(None=AppendVector(A=@n16, B=@n17) {desc=MAKE VECTOR3}) {transformSourceType=TRANSFORMSOURCE_Local}
  ) {constB=0}
  Add(A=@n11.Result, B=ParticleRandom())
  n18 = MaterialFunctionCall(
   UVs=TextureCoordinate()
   Texture Scale=AppendVector(
    A=ScalarParameter() {defaultValue=1, group=Bolt, parameterName=Bolt Texture U Scalar, sortPriority=1}
    B=ScalarParameter() {defaultValue=1, group=Bolt, parameterName=Bolt Texture V Scalar, sortPriority=2}
   )
  ) {materialFunction=&ScaleUVsByCenter}
  Add(A=@n18.UVs, B=ParticleRandom())
  n19 = Sine(None=ComponentMask(None=TextureCoordinate()) {r=true}) {period=2}
  n20 = TextureCoordinate() {vTiling=5}
  n21 = ParticleRandom()
  n22 = ScalarParameter() {defaultValue=1, group=Dissolve, parameterName=Dissolve Texture U Scalar, sortPriority=1}
  n23 = ScalarParameter() {defaultValue=1, group=Dissolve, parameterName=Dissolve Texture V Scalar, sortPriority=2}
  n24 = TextureSampleParameter2D(UVs=MaterialFunctionCall(UVs=Add(A=@n20, B=@n21), Texture Scale=AppendVector(A=@n22, B=@n23)) {materialFunction=&ScaleUVsByCenter}.UVs) {group=Dissolve, parameterName=Dissolve Tex, sortPriority=0, texture=&T_Inky_Smoke_Tile}
  n25 = Multiply(A=MaterialFunctionCall(
    Transition Phase=DynamicParameter() {defaultValue=<rgba 2 0 1 1>, paramNames [Width Base Scalar, Dissolve, WPO Noise Amplitude, Param4]}.Dissolve
    Height Texture=@n24.R
    Contrast=ScalarParameter() {group=Dissolve, parameterName=Dissolve Contrast}
   ) {materialFunction=&HeightLerp}.Results)
  LinearInterpolate(
   A=Step(
    Y=OneMinus(None=@n19)
    X=DynamicParameter() {defaultValue=<rgba 2 0 1 1>, paramNames [Width Base Scalar, Dissolve, WPO Noise Amplitude, Param4]}.Dissolve
   )
   B=@n25
   Alpha=ScalarParameter() {defaultValue=1, parameterName=Use Detail Dissolve}
  )
  n26 = ParticleColor()
  MaterialFunctionCall(Emissive=@n26.RGB) {materialFunction=&MF_FX_EmissiveAdjustments}
  n27 = Constant()
  n28 = ScalarParameter() {defaultValue=1, group=Bolt, parameterName=Bolt Pan Speed}
  n29 = TextureSampleParameter2D(UVs=Panner(
    Coordinate=@n18.UVs
    Speed=MaterialFunctionCall(X=@n27, Y=@n28) {materialFunction=&MakeFloat2}.Result
   ) {speedY=1}) {
   group=Bolt
   parameterName=Bolt Texture
   sortPriority=0
   texture=/Game/Effects/Fort_Effects/GameplayPulginsAssets/ToonLightning/MasterMaterial/V1/T_LightningSplineTexture
  }
  n30 = DepthFade(Opacity=@n25, FadeDistance=ScalarParameter() {defaultValue=30, group=Fades, parameterName=FadeDistance})
  n31 = TextureCoordinate()
  n32 = Multiply(A=Sine(None=ComponentMask(None=@n31) {g=true}) {period=2}) {constB=2}
  n33 = Multiply(A=Multiply(A=@n26.A, B=Saturate(None=@n32)), B=ShadingPathSwitch(Default=@n30, Deferred=@n30, Forward=@n25, Mobile=@n25)) {desc=TIP MASK}
  n34 = ScalarParameter() {defaultValue=1, group=Dissolve, parameterName=Dissolve Opacity, sortPriority=3}
  Multiply(
   A=Multiply(A=Multiply(A=@n29.R, B=@n33), B=LinearInterpolate(B=@n24.R, Alpha=@n34) {constA=1, constAlpha=1}) {desc=OPACITY NOISE BREAKUP}
   B=MaterialFunctionCall("FadeSensitivity (optional)"=ScalarParameter() {defaultValue=1, group=Fades, parameterName=FadeSensitivity}) {materialFunction=&FadeBasedOnViewAngle}.Result
  )
  Multiply(A=@n9, B=@n14)
 ]
}
