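# Physics asset for the TotalFlock backpack cosmetic mesh (M_MED_TotalFlock_Pack), a Battle Royale back bling accessory.
# It defines a single body on the pack bone, approximated by one short, rotated capsule offset slightly from the bone origin.
# It has no constraints, so the body stands alone with no jointed or ragdoll chain.
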PhysicsAsset /BRCosmetics/Accessories/FORT_Backpacks/Backpack_M_MED_TotalFlock/Meshes/M_MED_TotalFlock_Pack_PhysicsAsset
bodies {pack [{shapeType=Capsule, length=11.0736, radius=11.0318, center=<0 -0.7632 -2.6396>, rotation=<rot -4.9831 -90 -89.688>}]}
constraints []
